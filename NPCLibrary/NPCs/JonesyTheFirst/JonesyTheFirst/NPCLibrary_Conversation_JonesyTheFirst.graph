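# Conversation database for the Jonesy The First NPC, defining his greeting and the conversation branches that follow it.
# A single entry point leads to a speech node with a default line plus contextual lines.
# These are chosen by requirements for a first-ever conversation or the player's equipped character skin.
# The speech node then links out to shared Season 16 Spire Joneses, Basic Services and NPC Library Quests conversations by entry tag.
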ConversationDatabase /NPCLibrary/NPCs/JonesyTheFirst/JonesyTheFirst/NPCLibrary_Conversation_JonesyTheFirst
entry_points [{tag=Conversation.Entry.JonesyTheFirst, destinations [#1]}]
nodes [
 ConversationEntryPointNode_7 {
  id=#1
  class=ConversationEntryPointNode
  outputs [#2]
  properties {entryTag {tagName=Conversation.Entry.JonesyTheFirst}}
 }
 Speech_C_4 {
  id=#2
  class=Speech_C
  outputs [#3, #4, #5]
  properties {
   nodeName=""
   generalConfig {
    defaultMessage=I think being Jonesy the FIRST gives me 'leave my dirty dishes in the sink' privileges.
    contextualMessages [
     {
      message=It's hard being the best Jonesy.
      contextRequirements [{requirement {class=FortControllerRequirement_IsFirstEverConversationWithNPC}}]
      priority {curve {rowName=Tandem.Speech.Priority.FirstEverInteraction}}
     }
     {
      message="Look, over there! Bigfoot! You should... go there!"
      contextRequirements [{requirement {
         class=FortControllerRequirement_HasCID
         properties {allowedCIds [/BRCosmetics/Athena/Items/Cosmetics/Characters/CID_409_Athena_Commando_M_BunkerMan]}
        }}]
      priority {value=0, curve {curveTable=null, rowName=None}}
     }
     {
      message=Your mission would be child's play for ME.
      contextRequirements [{requirement {
         class=FortControllerRequirement_HasCID
         properties {allowedCIds [&CID_989_Athena_Commando_M_ProgressiveJonesy]}
        }}]
      priority {value=0, curve {curveTable=null, rowName=None}}
     }
     {
      message=Why didn't you just build a boat?
      contextRequirements [{requirement {
         class=FortControllerRequirement_HasCID
         properties {allowedCIds [/BRCosmetics/Athena/Items/Cosmetics/Characters/CID_836_Athena_Commando_M_JonesyFlare]}
        }}]
      priority {value=0, curve {curveTable=null, rowName=None}}
     }
     {
      message="Wait, why do YOU get a bathrobe?"
      contextRequirements [{requirement {
         class=FortControllerRequirement_HasCID
         properties {allowedCIds [/BRCosmetics/Athena/Items/Cosmetics/Characters/CID_798_Athena_Commando_M_JonesyVagabond]}
        }}]
      priority {value=0, curve {curveTable=null, rowName=None}}
     }
     {
      message="There he is, best of the rest!"
      contextRequirements [{requirement {
         class=FortControllerRequirement_HasCID
         properties {allowedCIds [&CID_883_Athena_Commando_M_ChOneJonesy]}
        }}]
      priority {value=0, curve {curveTable=null, rowName=None}}
     }
     {
      message="So you're rebellious? I can dig it. "
      contextRequirements [{requirement {class=FortControllerRequirement_HasCID, properties {allowedCIds [&Character_ChillCat]}}}]
      priority {value=0, curve {curveTable=null, rowName=None}}
     }
     {
      message="I'd give you some hard-won wisdom, but I got nothing."
      contextRequirements [{requirement {class=FortControllerRequirement_HasCID, properties {allowedCIds [&Character_Bites]}}}]
      priority {value=0, curve {curveTable=null, rowName=None}}
     }
     {
      message="Quality's not about the suit, it's about the bear inside the suit."
      contextRequirements [{requirement {class=FortControllerRequirement_HasCID, properties {allowedCIds [&Character_BadBear]}}}]
      priority {value=0, curve {curveTable=null, rowName=None}}
     }
     {
      message=I'm watching you.
      contextRequirements [{requirement {class=FortControllerRequirement_HasCID, properties {allowedCIds [&Character_TheHerald]}}}]
      priority {value=0, curve {curveTable=null, rowName=None}}
     }
     {
      message=You sure you want to get involved in all this?
      contextRequirements [{requirement {class=FortControllerRequirement_HasCID, properties {allowedCIds [&Character_RoseDust]}}}]
      priority {value=0, curve {curveTable=null, rowName=None}}
     }
     {
      message=Careful. This place can be disorienting.
      contextRequirements [{requirement {class=FortControllerRequirement_HasCID, properties {allowedCIds [&Character_PinkSpike]}}}]
      priority {value=0, curve {curveTable=null, rowName=None}}
     }
    ]
   }
  }
 }
 ConversationLinkNode_1 {id=#5, class=ConversationLinkNode, properties {remoteEntryTag {tagName=Conversation.Entry.NPCLibrary.Quests}}}
 ConversationLinkNode_0 {id=#4, class=ConversationLinkNode, properties {remoteEntryTag {tagName=Conversation.Entry.BasicServices}}}
 ConversationLinkNode_2 {
  id=#3
  class=ConversationLinkNode
  properties {remoteEntryTag {tagName=Conversation.Entry.Season16.TheSpire.Joneses}}
 }
]
